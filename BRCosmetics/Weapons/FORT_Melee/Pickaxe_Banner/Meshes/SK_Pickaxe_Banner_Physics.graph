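# Physics asset for the banner pickaxe skeletal mesh, a two-handed melee cosmetic with a simulatable hanging banner.
# A capsule body covers the handle, and six box bodies run along the dyn_banner bone chain, narrowing over the last two segments.
# Each banner body is constrained to the one before it, with the first anchored to the handle.
# All joints share the same tight swing and twist limits, keeping the banner's bend restrained.
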PhysicsAsset /BRCosmetics/Weapons/FORT_Melee/Pickaxe_Banner/Meshes/SK_Pickaxe_Banner_Physics
bodies {
 Melee_TwoHanded_Handle [{shapeType=Capsule, length=95.5808, radius=4.4883, center=<0.4303 -1.0609 35.2016>}]
 dyn_banner_01 [{shapeType=Box, extentX=6.6596, extentY=10.3344, extentZ=3.6533, center=<1.1318 0 -0.0001>}]
 dyn_banner_02 [{shapeType=Box, extentX=7.5, extentY=10.3344, extentZ=3.6533, center=<1.5445 0 -0.0001>}]
 dyn_banner_03 [{shapeType=Box, extentX=7.5, extentY=10.3344, extentZ=3.6533, center=<1.5445 0 -0.0001>}]
 dyn_banner_04 [{shapeType=Box, extentX=7.5, extentY=10.3344, extentZ=3.6533, center=<1.5445 0 -0.0001>}]
 dyn_banner_05 [{shapeType=Box, extentX=7.5, extentY=9.3652, extentZ=3.6533, center=<1.5445 0 -0.0001>}]
 dyn_banner_06 [{shapeType=Box, extentX=7.5, extentY=6.2061, extentZ=3.6533, center=<1.5445 0 -0.0001>}]
}
constraints [
 {bone1Name=dyn_banner_01, bone2Name=Melee_TwoHanded_Handle, swing1LimitDegrees=15, swing2LimitDegrees=30, twistLimitDegrees=10}
 {bone1Name=dyn_banner_02, bone2Name=dyn_banner_01, swing1LimitDegrees=15, swing2LimitDegrees=30, twistLimitDegrees=10}
 {bone1Name=dyn_banner_03, bone2Name=dyn_banner_02, swing1LimitDegrees=15, swing2LimitDegrees=30, twistLimitDegrees=10}
 {bone1Name=dyn_banner_04, bone2Name=dyn_banner_03, swing1LimitDegrees=15, swing2LimitDegrees=30, twistLimitDegrees=10}
 {bone1Name=dyn_banner_05, bone2Name=dyn_banner_04, swing1LimitDegrees=15, swing2LimitDegrees=30, twistLimitDegrees=10}
 {bone1Name=dyn_banner_06, bone2Name=dyn_banner_05, swing1LimitDegrees=15, swing2LimitDegrees=30, twistLimitDegrees=10}
]
